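# Material function for Apollo landscape materials that selects between runtime virtual texture (RVT) material attributes and a bypass set.
# The EnableRVT static switch, on by default, passes the RVT input, but only when the virtual texture feature is supported.
# An optional debug switch blends both inputs by rounded, channel-masked viewport UV for a split-screen comparison, marked not to be saved enabled.
# Material proxy generation always uses the bypass input.
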MaterialFunction /Game/Athena/Apollo/Maps/Landscape/VT/MaterialFunctions/EnableRVT
graph {nodes [
  n1 = Round(None=ChannelMaskParameter(None=ScreenPosition().ViewportUV).R)
  n2 = FunctionInput() {inputName=ByPass RVT, inputType=FunctionInput_MaterialAttributes, sortPriority=1}
  n3 = FunctionInput() {inputName=RVT, inputType=FunctionInput_MaterialAttributes}
  n4 = StaticSwitchParameter(
   True=StaticSwitchParameter(True=BlendMaterialAttributes(A=@n3, B=@n2, Alpha=@n1), False=@n3) {desc="Split screen compare (don't save with this on!)", group=RVT, parameterName=EnableRVTDebugCompare}
   False=@n2
  ) {defaultValue=true, group=RVT, parameterName=EnableRVT}
  FunctionOutput(None=MaterialProxyReplace(Realtime=VirtualTextureFeatureSwitch!(No=@n2, Yes=@n4), MaterialProxy=@n2))
 ]}
